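; Gameplay cue burst notify that plays impact effects for the Wasabi claws' spin slash hit.
; The Cascade burst reads the hit result and acts only when the hit actor is a non-pawn vehicle and the effect causer is a weapon.
; It spawns impact FX and an optional fading decal, aimed from weapon owner to impact, with roll offset scaled by the owner's facing dot product.
; The Niagara burst is empty.
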(fn OnBurst(Cascade)
  (params (MyTarget :type Actor%20Object%20Reference) (Parameters :type Gameplay%20Cue%20Parameters%20Structure%20%28by%20ref%29) (ParticleComponents :type Array%20of%20Cascade%20Particle%20System%20Component%20Object%20References) (AudioComponents :type Array%20of%20Audio%20Component%20Object%20References) (BurstCameraShakeInstance :type Camera%20Shake%20Base%20Object%20Reference) (BurstDecalInstance :type Decal%20Actor%20Object%20Reference))
  (bind (_normalizedmagnitude _rawmagnitude _effectcontext) (Ability|GameplayCue|BreakGameplayCueParameters Parameters))
  (bind _returnvalue (Ability|EffectContext|GetHitResult _effectcontext))
  (bind (_bblockinghit _binitialoverlap _time _distance _location _impactpoint _normal _impactnormal _physmat _hitactor) (Collision|BreakHitResult _returnvalue))
  (bind _localimpactlocation (LocalVariables|GetLocalImpactLocation))
  (bind _localimpactlocation_2 (LocalVariables|GetLocalImpactLocation))
  (bind _impactrotation (Variables|Effects|GetImpactRotation))
  (LocalVariables|SetLocalImpactLocation _location)
  (LocalVariables|SetLocalHitActor _hitactor)
  ;[1 1 1] Hit FX + Decal
  (Utilities|Casting|CastToFortPawn (LocalVariables|GetLocalHitActor)
    (:then
      _)
    (:CastFailed
      (Utilities|Casting|CastToFortVehicleInterface (LocalVariables|GetLocalHitActor)
        (:then
          (bind _asfort_weapon (Utilities|Casting|CastToFortWeapon (Ability|EffectContext|GetEffectCauser _effectcontext))
            (:then
              (bind (_returnvalue_roll _returnvalue_pitch _returnvalue_yaw) (Math|Rotator|BreakRotator (Math|Rotator|FindLookatRotation (Transformation|GetActorLocation (Actor|GetOwner _asfort_weapon)) _localimpactlocation)))
              (Effects|Components|ParticleSystem|SpawnFXAtLocation (LocalVariables|GetLocalHitActor) (Variables|Effects|GetImpactFX) (LocalVariables|GetLocalImpactLocation) (Math|Rotator|MakeRotator (+ _returnvalue_roll (* _impactrotation (Math|Vector|DotProduct (Transformation|GetActorForwardVector (Actor|GetOwner _asfort_weapon)) (Math|Vector|Normalize (- _localimpactlocation_2 (Transformation|GetActorLocation (Actor|GetOwner _asfort_weapon))))))) _returnvalue_pitch _returnvalue_yaw) "1.000000,1.000000,1.000000" true "AutoRelease")
              (if (Variables|Default|GetShouldSpawnDecal)
                (bind _returnvalue_2 (Rendering|Decal|SpawnDecalatLocation (LocalVariables|GetLocalHitActor) (Variables|Effects|GetImpactDecal) (Variables|Effects|GetDecalScale) (LocalVariables|GetLocalImpactLocation) (Math|Rotator|MakeRotator (+ _returnvalue_roll (* _impactrotation (Math|Vector|DotProduct (Transformation|GetActorForwardVector (Actor|GetOwner _asfort_weapon)) (Math|Vector|Normalize (- _localimpactlocation_2 (Transformation|GetActorLocation (Actor|GetOwner _asfort_weapon))))))) _returnvalue_pitch _returnvalue_yaw) 2.0))
                (Rendering|Components|Decal|SetFadeOut _returnvalue_2 0.0 2.0 false)
                (else
                  _))
              (return))
            (:CastFailed
              _)))
        (:CastFailed
          _)))))

(fn OnBurst(Niagara)
  (params (MyTarget :type Actor%20Object%20Reference) (Parameters :type Gameplay%20Cue%20Parameters%20Structure%20%28by%20ref%29) (ParticleComponents :type Array%20of%20Niagara%20Particle%20System%20Component%20Object%20References) (AudioComponents :type Array%20of%20Audio%20Component%20Object%20References) (BurstCameraShakeInstance :type Camera%20Shake%20Base%20Object%20Reference) (BurstDecalInstance :type Decal%20Actor%20Object%20Reference))
  (return))
